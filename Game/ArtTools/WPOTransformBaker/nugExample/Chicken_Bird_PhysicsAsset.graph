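# Physics asset for the Chicken_Bird skeleton in the WPOTransformBaker example folder, giving the bird a simplified capsule body for collision and physics simulation.
# One capsule each covers the pelvis, spine, chest, neck base, head, both elbows, both knees and both toes.
# Constraints chain pelvis to head, attach elbows to the chest and knees to the pelvis, and link each toe to its knee.
# Every joint shares the same uniform swing and twist limit angle.
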PhysicsAsset /Game/ArtTools/WPOTransformBaker/nugExample/Chicken_Bird_PhysicsAsset
bodies {
 C_BipedHeadNeck_A_Head_Jnt [{shapeType=Capsule, length=21.3388, radius=11.2916, center=<9.3272 0.0166 -0.0011>, rotation=<rot 0 -124.1134 -89.9947>}]
 C_BipedHeadNeck_A_NeckBase_Jnt [{shapeType=Capsule, length=10.4292, radius=8.1155, center=<11.7961 -1.7077 0.0004>, rotation=<rot 0.0076 -90 0.0163>}]
 C_QuadSpine_A_Chest_Jnt [{shapeType=Capsule, length=21.8365, radius=17.9692, center=<0.8721 3.7848 -0.4103>, rotation=<rot -0.6101 -90 1.6856>}]
 C_QuadSpine_A_Pelvis_Jnt [{shapeType=Capsule, length=16.6811, radius=16.0852, center=<-5.5201 -3.7255 1.7201>, rotation=<rot 34.559 -90 6.4641>}]
 C_QuadSpine_A_Spine2_Jnt [{shapeType=Capsule, length=19.2119, radius=16.217, center=<1.4289 -0.3076 0.0021>, rotation=<rot 0 -173.1358 -89.9683>}]
 L_BipedArm_A_Elbow_Jnt [{shapeType=Capsule, length=14.3825, radius=9.0845, center=<8.8082 0.4912 0.4889>, rotation=<rot 0 -89.6042 -57.3743>}]
 L_HoofedLeg_A_Knee_Jnt [{shapeType=Capsule, length=13.8293, radius=11.6798, center=<1.1806 2.3681 1.0726>, rotation=<rot 0 -123.3201 -72.3581>}]
 L_HoofedLeg_A_Toe1_Jnt [{shapeType=Capsule, length=11.1259, radius=8.8517, center=<2.263 0.4794 -0.5731>, rotation=<rot 0 97.51 -60.2776>}]
 R_BipedArm_A_Elbow_Jnt [{shapeType=Capsule, length=14.3778, radius=9.0928, center=<-8.8443 -0.4828 -0.4946>, rotation=<rot 0 -89.7639 -57.6542>}]
 R_HoofedLeg_A_Knee_Jnt [{shapeType=Capsule, length=13.9021, radius=11.4848, center=<-0.565 -2.7569 0.5558>, rotation=<rot 0 -124.6633 -73.1429>}]
 R_HoofedLeg_A_Toe1_Jnt [{shapeType=Capsule, length=11.1259, radius=8.8517, center=<-2.263 -0.4794 0.5731>, rotation=<rot 0 97.51 -60.2775>}]
}
constraints [
 {bone1Name=C_QuadSpine_A_Spine2_Jnt, bone2Name=C_QuadSpine_A_Pelvis_Jnt, swing1LimitDegrees=45, swing2LimitDegrees=45, twistLimitDegrees=45}
 {bone1Name=C_QuadSpine_A_Chest_Jnt, bone2Name=C_QuadSpine_A_Spine2_Jnt, swing1LimitDegrees=45, swing2LimitDegrees=45, twistLimitDegrees=45}
 {
  bone1Name=C_BipedHeadNeck_A_NeckBase_Jnt
  bone2Name=C_QuadSpine_A_Chest_Jnt
  swing1LimitDegrees=45
  swing2LimitDegrees=45
  twistLimitDegrees=45
 }
 {
  bone1Name=C_BipedHeadNeck_A_Head_Jnt
  bone2Name=C_BipedHeadNeck_A_NeckBase_Jnt
  swing1LimitDegrees=45
  swing2LimitDegrees=45
  twistLimitDegrees=45
 }
 {bone1Name=L_BipedArm_A_Elbow_Jnt, bone2Name=C_QuadSpine_A_Chest_Jnt, swing1LimitDegrees=45, swing2LimitDegrees=45, twistLimitDegrees=45}
 {bone1Name=R_BipedArm_A_Elbow_Jnt, bone2Name=C_QuadSpine_A_Chest_Jnt, swing1LimitDegrees=45, swing2LimitDegrees=45, twistLimitDegrees=45}
 {bone1Name=L_HoofedLeg_A_Knee_Jnt, bone2Name=C_QuadSpine_A_Pelvis_Jnt, swing1LimitDegrees=45, swing2LimitDegrees=45, twistLimitDegrees=45}
 {bone1Name=L_HoofedLeg_A_Toe1_Jnt, bone2Name=L_HoofedLeg_A_Knee_Jnt, swing1LimitDegrees=45, swing2LimitDegrees=45, twistLimitDegrees=45}
 {bone1Name=R_HoofedLeg_A_Knee_Jnt, bone2Name=C_QuadSpine_A_Pelvis_Jnt, swing1LimitDegrees=45, swing2LimitDegrees=45, twistLimitDegrees=45}
 {bone1Name=R_HoofedLeg_A_Toe1_Jnt, bone2Name=R_HoofedLeg_A_Knee_Jnt, swing1LimitDegrees=45, swing2LimitDegrees=45, twistLimitDegrees=45}
]
